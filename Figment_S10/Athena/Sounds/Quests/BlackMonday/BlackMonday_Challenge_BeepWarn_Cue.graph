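# Sound cue for the warning beep of the Black Monday quest challenge in Figment Season 10.
# A single wave player plays the BeepWarn sound wave with a boosted volume, with no randomization or mixing.
# It routes through the UI sound class and uses the quest-interactable hint attenuation, limiting audibility to a short range around the source.
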SoundCue /Figment_S10/Athena/Sounds/Quests/BlackMonday/BlackMonday_Challenge_BeepWarn_Cue
root=WavePlayer_0
nodes [{id=WavePlayer_0, properties {soundWaveAssetPtr=&/BlackMonday_Challenge_BeepWarn}}]
properties {
 attenuationSettings=/Game/Sounds/Attenuation/Fort_QuestInteractables_Hint
 maxDistance=1300
 soundClassObject=&Fort_UI
 volumeMultiplier=5
}
