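# BlackMonday_Challenge_Explo_Cue is a SoundCue for an explosion sound in the Black Monday quest challenge set.
# A single Wave Player plays the BlackMonday_Challenge_Explo wave directly, with no randomization, mixing or modulation.
# It uses the close-range explosion attenuation asset for spatial falloff, yet is routed through the UI sound class rather than an effects class.
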SoundCue /Figment_S10/Athena/Sounds/Quests/BlackMonday/BlackMonday_Challenge_Explo_Cue
root=WavePlayer_0
nodes [{id=WavePlayer_0, properties {soundWaveAssetPtr=&/BlackMonday_Challenge_Explo}}]
properties {
 attenuationSettings=/Figment_S10/Sounds/Attenuation/Fort_Explosion_Close_Att
 maxDistance=3500
 soundClassObject=&Fort_UI
 volumeMultiplier=1
}
