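# Sound cue for the looping audio of the Black Monday quest challenge in Figment Season 10.
# A single looping wave player plays the BlackMonday_Challenge_Loop wave as the cue's root.
# It uses the shared quest-interactable hint attenuation with a limited audible range and is routed through the Fort_UI sound class.
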SoundCue /Figment_S10/Athena/Sounds/Quests/BlackMonday/BlackMonday_Challenge_Loop_Cue
root=WavePlayer_0
nodes [{id=WavePlayer_0, properties {bLooping=true, soundWaveAssetPtr=&/BlackMonday_Challenge_Loop}}]
properties {
 attenuationSettings=/Game/Sounds/Attenuation/Fort_QuestInteractables_Hint
 maxDistance=1300
 soundClassObject=&Fort_UI
 volumeMultiplier=1
}
