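# Sound cue for the BlackMonday quest "turn off" sound, routed to the Fort_Quests sound class.
# A single wave player feeds BlackMonday_TurnOff straight to the output at boosted volume, with no randomization or mixing.
# Inline attenuation overrides take precedence over the referenced quest-interactable hint attenuation asset, disabling spatialization while setting a distance falloff so the sound stays local without panning.
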SoundCue /Figment_S10/Athena/Sounds/Quests/BlackMonday/BlackMonday_TurnOff_Cue
root=WavePlayer_2
nodes [{id=WavePlayer_2, properties {soundWaveAssetPtr=&/BlackMonday_TurnOff}}]
properties {
 attenuationSettings=/Game/Sounds/Attenuation/Fort_QuestInteractables_Hint
 bOverrideAttenuation=true
 maxDistance=1700
 soundClassObject=&Fort_Quests
 volumeMultiplier=1.5
 attenuationOverrides {bSpatialize=false, falloffDistance=1000, stereoSpread=5000, attenuationShapeExtents {x=700}}
}
