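# Sound cue for the turn-on sound of the BlackMonday quest in Figment_S10, assigned to the Fort_Quests sound class.
# A single wave player feeds an attenuation node that uses the shared quest-interactable hint attenuation asset.
# That branch enters a mixer at boosted volume, with the mixer's second input left empty, and the cue adds its own volume boost.
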SoundCue /Figment_S10/Athena/Sounds/Quests/BlackMonday/BlackMonday_TurnOn_Cue
root=Mixer_0
nodes [
 {id=Attenuation_0, children [WavePlayer_0], properties {attenuationSettings=/Game/Sounds/Attenuation/Fort_QuestInteractables_Hint}}
 {id=Mixer_0, children [Attenuation_0, null], properties {inputVolume [1.5,1]}}
 {id=WavePlayer_0, properties {soundWaveAssetPtr=&/BlackMonday_TurnOn}}
]
properties {
 maxDistance=1300
 soundClassObject=&Fort_Quests
 volumeMultiplier=1.5
 attenuationOverrides {bSpatialize=false, falloffDistance=1000, stereoSpread=5000, attenuationShapeExtents {x=700}}
}
